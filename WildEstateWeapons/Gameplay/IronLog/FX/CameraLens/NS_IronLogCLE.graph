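# Camera lens Niagara effect for the IronLog weapon that shows a screen-edge sprite overlay in front of the camera.
# One emitter burst-spawns offset, camera-facing sprites, sized from a system base scale with a brief grow-in.
# Alpha follows a system-age ramp scaled by User.EffectIntensity, while curves animate dynamic material parameters.
# A system Timeline's normalized time fades the sprites out and kills them near its end.
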NiagaraSystem /WildEstateWeapons/Gameplay/IronLog/FX/CameraLens/NS_IronLogCLE
dependencies {
 usedDataInterfaces [&NiagaraDataInterfaceCurve]
 usedDynamicInputs [
  &RemapRange
  &Subtract_Float
  &Divide_Float
  /Niagara/DynamicInputs/Math/Min_Float
  &Multiply_Vector2D
 ]
 usedModules [
  &self:SystemSpawnScript.SystemScriptSource.SystemScriptGraph.NiagaraNodeAssignment_1.SetVariables_#1
  &self:SystemSpawnScript.SystemScriptSource.SystemScriptGraph.NiagaraNodeAssignment_2.SetVariables_#2
  &self:SystemSpawnScript.SystemScriptSource.SystemScriptGraph.NiagaraNodeAssignment_3.SetVariables_#3
  &self:SystemSpawnScript.SystemScriptSource.SystemScriptGraph.NiagaraNodeAssignment_6.SetVariables_#4
  &self:SystemSpawnScript.SystemScriptSource.SystemScriptGraph.NiagaraNodeAssignment_4.SetVariables_#5
  &self:SystemSpawnScript.SystemScriptSource.SystemScriptGraph.NiagaraNodeAssignment_7.SetVariables_#6
  &SystemState
  &Timeline
  &self:SystemSpawnScript.SystemScriptSource.SystemScriptGraph.NiagaraNodeAssignment_0.SetVariables_#7
  /Niagara/Modules/System/CompleteIfUnused
 ]
}
emitters {Screen_Edge {
  emitterUpdateScript [
   &EmitterState()
   &SpawnBurst_Instantaneous(Use Spawn Probability=false, UseLoopCountLimit=true)
   &SpawnBurst_Instantaneous(Spawn Time=9.8, Use Spawn Probability=false, UseLoopCountLimit=false) {module=SpawnBurst_Instantaneous001}
  ]
  particleSpawnScript [
   &InitializeParticle(
    Lifetime=$User.EffectLength
    Position Offset=<10.35 0 0>
    Sprite Size=<1 1>
    Sprite Size Mode=::NewEnumerator4
    UsePositionOffset=true
    Write Material Random=true
   )
   &SpriteFacingAndAlignment(Sprite Alignment=<0 0 1>, Sprite Facing=<-1 0 0>, WriteAlignment=true, WriteFacing=true)
   &CameraOffset()
  ]
  particleUpdateScript [
   &ParticleState(Kill Particles When Lifetime Has Elapsed=false, Let Infinitely Lived Particles Die When Emitter Deactivates=true)
   &ScaleColor(
    Scale Alpha=&FloatFromCurve(
     FloatCurve=DataInterface() {propertyValues {Curve {keys [
         <at 0 0 RCIM_Cubic>
         {arriveTangentWeight=0.0501, interpMode=RCIM_Cubic, leaveTangentWeight=0.4064, tangentMode=RCTM_User, time=0.15, value=1}
         {arriveTangentWeight=0.4027, interpMode=RCIM_Cubic, leaveTangentWeight=0.0474, tangentMode=RCTM_User, time=0.65, value=1}
        ]}}}
     CurveIndex=$Engine.System.Age
     Scale Curve=&Multiply_Float(
      A=&FloatFromCurve(
       FloatCurve=DataInterface() {propertyValues {Curve {defaultValue=1, keys [<at 0 0 RCIM_Cubic>, {arriveTangentWeight=0.5507, leaveTangentWeight=4.9567, time=1, value=1}, <at 10 3 RCIM_Cubic>]}}}
       CurveIndex=$User.EffectIntensity
      )
      B=2
     )
    )
    ScaleA=true
    ScaleRGB=true
   ) {module=ScaleColor001}
   &DynamicMaterialParameters(
    Index 0 Param 1=3
    Index 0 Param 2=0.1
    Index 0 Param 3=1
    Index 0 Param 4=0.1
    Module.Index0Param0Write=true
    Module.Index0Param1Write=true
    Module.Index0Param2Write=true
    Module.Index0Param3Write=true
   ) {enabled=false}
   &DynamicMaterialParameters(
    Index 0 Param 1=&FloatFromCurve(
     FloatCurve=DataInterface() {propertyValues {Curve {
        defaultValue=1
        keys [
         {arriveTangent=0.8128, interpMode=RCIM_Cubic, leaveTangent=0.8128, tangentMode=RCTM_User, time=0, value=0}
         {arriveTangent=1.3952, interpMode=RCIM_Cubic, leaveTangent=1.3952, tangentMode=RCTM_User, time=0.75, value=0.9}
         <at 1 1 RCIM_Cubic>
        ]
       }}}
     Scale Curve=0.1
    )
    Index 0 Param 2=0.1
    Index 0 Param 3=1
    Index 0 Param 4=&FloatFromCurve(FloatCurve=DataInterface() {propertyValues {Curve {keys [<at 0.65 1>,<at 1 0>]}}})
    Module.Index0Float0Write=true
    Module.Index0Float1Write=true
    Module.Index0Float2Write=true
    Module.Index0Float3Write=true
   ) {module=DynamicMaterialParameters001}
   /SetVariables(Particles.SpriteSize=&Multiply_Vector2D_ByFloat(
     Float=&FloatFromCurve(
      FloatCurve=DataInterface() {propertyValues {Curve {defaultValue=0, keys [{tangentMode=RCTM_User, time=0, value=0.75}, {tangentMode=RCTM_User, time=0.1, value=1}]}}}
      CurveIndex=$Emitter.Age
     )
     Vector2D=&Multiply_Vector2D_ByFloat(Float=2, Vector2D=$System.Base Scale)
    )) {isSetParameters=true}
   &ScaleColor(Scale Alpha=&OneMinusFloat(Float=$System.Timeline.NormalizedTime), ScaleA=true, ScaleRGB=true)
   &KillParticles(Kill Particles=&SetBoolByFloatComparison(A=$System.Timeline.NormalizedTime, B=0.99))
  ]
  rendererClasses [&NiagaraSpriteRendererProperties]
  renderers [{rendererClass=&NiagaraSpriteRendererProperties}]
 }}
summary {
 systemName=NS_IronLogCLE
 emitters [{emitterName=Screen_Edge, rendererClasses [&NiagaraSpriteRendererProperties]}]
 userVariables [
  User.EffectIntensity {type=NiagaraFloat, defaultValue=1, description=""}
  User.EffectLength {type=NiagaraFloat, defaultValue=4, description=""}
  User.Ratio {type=Vector2f, description="", defaultValue {struct=Vector2f, value=<1 1>}}
  User.Test {type=NiagaraFloat, defaultValue=5, description=""}
 ]
}
